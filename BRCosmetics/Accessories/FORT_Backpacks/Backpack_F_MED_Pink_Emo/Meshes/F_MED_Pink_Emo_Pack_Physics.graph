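# Physics asset for the Pink Emo backpack cosmetic, defining bodies for its dangling heart, strap and zipper bones.
# Each bone has one small capsule, with only the zipper's offset and rotated.
# Both hearts, the first strap segment and the zipper constrain to dyn_heart_root, while the other strap segments chain from the first.
# All joints share tight, uniform swing and twist limits, restricting angular movement to slight motion.
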PhysicsAsset /BRCosmetics/Accessories/FORT_Backpacks/Backpack_F_MED_Pink_Emo/Meshes/F_MED_Pink_Emo_Pack_Physics
bodies {
 dyn_heart_1 [{shapeType=Capsule, length=2, radius=2, shapeName=dyn_heart_1_capsule}]
 dyn_heart_2 [{shapeType=Capsule, length=2, radius=2, shapeName=dyn_heart_2_capsule}]
 dyn_heart_root [{shapeType=Capsule, length=1, radius=1, shapeName=dyn_heart_root_capsule}]
 dyn_strap_1 [{shapeType=Capsule, length=2, radius=2, shapeName=dyn_strap_1_capsule}]
 dyn_strap_2 [{shapeType=Capsule, length=2, radius=2, shapeName=dyn_strap_2_capsule}]
 dyn_strap_3 [{shapeType=Capsule, length=2, radius=2, shapeName=dyn_strap_3_capsule}]
 dyn_zipper_1 [{shapeType=Capsule, length=4.2213, radius=0.8135, shapeName=dyn_zipper_1_capsule, center=<3.5216 -0.2371 0>, rotation=<rot 90 0 -7.4372>}]
}
constraints [
 {bone1Name=dyn_strap_2, bone2Name=dyn_strap_1, swing1LimitDegrees=5, swing2LimitDegrees=5, twistLimitDegrees=5}
 {bone1Name=dyn_strap_3, bone2Name=dyn_strap_2, swing1LimitDegrees=5, swing2LimitDegrees=5, twistLimitDegrees=5}
 {bone1Name=dyn_heart_2, bone2Name=dyn_heart_root, swing1LimitDegrees=5, swing2LimitDegrees=5, twistLimitDegrees=5}
 {bone1Name=dyn_heart_1, bone2Name=dyn_heart_root, swing1LimitDegrees=5, swing2LimitDegrees=5, twistLimitDegrees=5}
 {bone1Name=dyn_strap_1, bone2Name=dyn_heart_root, swing1LimitDegrees=5, swing2LimitDegrees=5, twistLimitDegrees=5}
 {bone1Name=dyn_zipper_1, bone2Name=dyn_heart_root, swing1LimitDegrees=5, swing2LimitDegrees=5, twistLimitDegrees=5}
]
